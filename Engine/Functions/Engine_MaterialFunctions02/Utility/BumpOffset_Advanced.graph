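# BumpOffset_Advanced is a library material function that offsets UVs by height to fake parallax depth.
# It multiplies the camera vector's XY by height minus a reference plane, then by a height ratio.
# The camera vector defaults to tangent space but can be overridden, for example with world space for world-space UVs.
# Outputs are offset UVs, the raw distortion and the undeformed UVs.
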MaterialFunction /Engine/Functions/Engine_MaterialFunctions02/Utility/BumpOffset_Advanced.BumpOffset_advanced
graph {
 properties {bExposeToLibrary=true, libraryCategoriesText [Misc,Texturing]}
 nodes [
  n1 = FunctionInput(Preview=Transform(None=CameraVectorWS()) {transformSourceType=TRANSFORMSOURCE_World, transformType=TRANSFORM_Tangent}) {
   bUsePreviewValueAsDefault=true
   description=By default the camera vector in tangent space but use world space if the uv coordinates are in world space.
   inputName=CameraVector
   sortPriority=10
   previewValue=<0.5 0 0 1>
  }
  n2 = FunctionInput() {bUsePreviewValueAsDefault=true, inputName=HeightTexture, inputType=FunctionInput_Scalar}
  n3 = FunctionInput() {bUsePreviewValueAsDefault=true, inputName=ReferencePlane, inputType=FunctionInput_Scalar, sortPriority=3, previewValue=<0.5 0 0 1>}
  n4 = FunctionInput() {bUsePreviewValueAsDefault=true, inputName=HeightRatioInput, inputType=FunctionInput_Scalar, sortPriority=4, previewValue=<0.5 0 0 1>}
  n5 = Multiply(A=Multiply(A=ComponentMask(None=@n1) {g=true, r=true}, B=Subtract(A=@n2, B=@n3)), B=@n4) {constB=0.1}
  n6 = FunctionInput(Preview=TextureCoordinate()) {bUsePreviewValueAsDefault=true, inputName=TextureCoordinates, inputType=FunctionInput_Vector2, sortPriority=1, previewValue=<0.5 0 0 1>}
  FunctionOutput(None=Add(A=@n5, B=@n6) {desc=+UVs}) {outputName=UVs}
  FunctionOutput(None=@n5) {description=Add uvs to the distortion before using this value as uvs., outputName=Distortion, sortPriority=1}
  FunctionOutput(None=@n6) {outputName=Undeformed UVs, sortPriority=-1}
 ]
}
