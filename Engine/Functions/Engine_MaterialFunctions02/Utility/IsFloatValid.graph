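# IsFloatValid is a library-exposed utility material function that returns 1 when a scalar input is a real number and 0 when it is infinite or NaN.
# The single scalar input feeds a Custom HLSL node that tests it with isnan and isinf, and that node's float result goes straight to the function output.
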MaterialFunction /Engine/Functions/Engine_MaterialFunctions02/Utility/IsFloatValid
graph {
 properties {bExposeToLibrary=true, description=Returns 1 if the float is a real number and 0 if it is either infinite or a NaN.}
 nodes [
  MaterialExpressionFunctionInput_0 = FunctionInput() {inputType=FunctionInput_Scalar}
  n1 = Custom!(X=@MaterialExpressionFunctionInput_0) {
   code="float myFloatOutput = 1.0;\r\n\r\n\r\nif (isnan(X) || isinf(X)) {\r\nmyFloatOutput = 0.0f;\r\n}\r\nreturn myFloatOutput;"
   outputType=CMOT_Float1
   inputs [{inputName=X, input {expression=&self:MaterialExpressionFunctionInput_0}}]
  }
  FunctionOutput(None=@n1)
 ]
}
