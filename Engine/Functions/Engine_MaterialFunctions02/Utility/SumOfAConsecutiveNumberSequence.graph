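# Material function, exposed to the library under Misc and Math, that returns the sum 1+2+...+n for a scalar input n.
# It takes the absolute value of the input, which falls back to its preview value when unconnected, then applies the closed-form triangular number formula n(n+1)/2 without a loop.
# The Add and Divide nodes rely on their default constants for the +1 and the halving.
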MaterialFunction /Engine/Functions/Engine_MaterialFunctions02/Utility/SumOfAConsecutiveNumberSequence
graph {
 properties {
  bExposeToLibrary=true
  description="This will find the sum of numbers up to the integer inserted. I.e. [ 1+2+3...+ n ]."
  libraryCategoriesText [Misc,Math]
 }
 nodes [
  n1 = FunctionInput() {bUsePreviewValueAsDefault=true, inputName=Number of integers, inputType=FunctionInput_Scalar, previewValue=<1 0 0 1>}
  n2 = Abs(None=@n1)
  n3 = Add(A=@n2)
  FunctionOutput(None=Divide(A=Multiply(A=@n3, B=@n2)))
 ]
}
