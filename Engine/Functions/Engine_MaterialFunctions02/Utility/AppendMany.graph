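# AppendMany is a library material function that packs up to four scalars into vectors in a single node.
# Scalar inputs R, G, B and A feed a chain of three AppendVector nodes, each adding one channel to the previous result.
# Each stage has its own output, giving RG, RGB and RGBA, and every input uses its preview value as default.
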MaterialFunction /Engine/Functions/Engine_MaterialFunctions02/Utility/AppendMany
graph {
 properties {bExposeToLibrary=true, description=Allows you to quickly append up to 4 scalars in one function, libraryCategoriesText [Misc,VFX]}
 nodes [
  n1 = AppendVector(
   A=FunctionInput() {bUsePreviewValueAsDefault=true, inputName=R, inputType=FunctionInput_Scalar}
   B=FunctionInput() {bUsePreviewValueAsDefault=true, inputName=G, inputType=FunctionInput_Scalar, sortPriority=1}
  ) {desc=RG}
  FunctionOutput(None=@n1) {outputName=RG}
  n2 = AppendVector(A=@n1, B=FunctionInput() {bUsePreviewValueAsDefault=true, inputName=B, inputType=FunctionInput_Scalar, sortPriority=2}) {desc=RGB}
  FunctionOutput(None=@n2) {outputName=RGB, sortPriority=1}
  FunctionOutput(None=AppendVector(A=@n2, B=FunctionInput() {bUsePreviewValueAsDefault=true, inputName=A, inputType=FunctionInput_Scalar, sortPriority=3}) {desc=RGBA}) {outputName=RGBA, sortPriority=2}
 ]
}
