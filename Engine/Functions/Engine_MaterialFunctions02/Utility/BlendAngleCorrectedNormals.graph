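# Material function that blends an additional normal map, typically a detail map, over a base normal with corrected angles, exposed in the library under Misc and Normals.
# It offsets the base normal's Z by the Add node's default constant and flips the additional normal's XY.
# A dot product of the two scales the base vector, the additional vector scaled by offset Z is subtracted, and the result is normalized, matching Reoriented Normal Mapping.
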MaterialFunction /Engine/Functions/Engine_MaterialFunctions02/Utility/BlendAngleCorrectedNormals
graph {
 properties {
  bExposeToLibrary=true
  description="Corrects the normal direction of normal map that is overlayed on other normal maps. (Typically this would be a detail map.) "
  libraryCategoriesText [Misc,Normals]
 }
 nodes [
  n1 = FunctionInput() {inputName=BaseNormal}
  n2 = Add(A=ComponentMask(None=@n1) {b=true})
  n3 = AppendVector(A=ComponentMask(None=@n1) {g=true, r=true}, B=@n2)
  n4 = FunctionInput() {description=Insert a tiling normal map or other normal map to blend with the base normal., inputName=AdditionalNormal, sortPriority=1}
  n5 = AppendVector(A=Multiply(A=ComponentMask(None=@n4) {g=true, r=true}) {constB=-1}, B=ComponentMask(None=@n4) {b=true})
  n6 = Multiply(A=@n3, B=DotProduct(A=@n3, B=@n5))
  n7 = Multiply(A=@n2, B=@n5)
  FunctionOutput(None=Normalize(VectorInput=Subtract(A=@n6, B=@n7)))
 ]
}
