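# Material function, exposed to the library, that combines a detail normal with a base normal in world space, correcting the detail normal's direction.
# A Custom HLSL node builds a shortest-arc quaternion from the geometric VertexNormalWS to the BaseNormal input, then rotates the AdditionalNormal input by it.
# The math follows the reoriented normal blending technique from the selfshadow "Blending in Detail" article, and the rotated vector is the only output.
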MaterialFunction /Engine/Functions/Engine_MaterialFunctions02/Utility/BlendAngleCorrectedNormals_WS
graph {
 properties {
  bExposeToLibrary=true
  description=Corrects the normal direction of normal map that is combined with another normal map. This function works with world space vectors.
 }
 nodes [
  MaterialExpressionFunctionInput_1 = FunctionInput(Preview=Constant3Vector() {constant=<rgba 1 0 0 0>}) {bUsePreviewValueAsDefault=true, description=detail normal, inputName=AdditionalNormal, sortPriority=1}
  MaterialExpressionFunctionInput_0 = FunctionInput(Preview=Constant3Vector() {constant=<rgba 0 0 1 0>}) {bUsePreviewValueAsDefault=true, inputName=BaseNormal}
  MaterialExpressionVertexNormalWS_2 = VertexNormalWS() {desc=geometric normal}
  n1 = Custom!(u=@MaterialExpressionFunctionInput_1, t=@MaterialExpressionFunctionInput_0, s=@MaterialExpressionVertexNormalWS_2) {
   code="//https://blog.selfshadow.com/publications/blending-in-detail/ (equation 4 in nitty-gritty)\r\n// Build the shortest-arc quaternion\r\nfloat4 q = float4(cross(s, t), dot(s, t) + 1) / sqrt(2 * (dot(s, t) + 1));\r\n// Rotate the normal\r\nreturn u * (q.w * q.w - dot(q.xyz, q.xyz)) + 2 * q.xyz * dot(q.xyz, u) + 2 * q.w * cross(q.xyz, u);\r\n"
   description=BlendAngleCorrectedWorld
   showCode=true
   inputs [
    {
     inputName=u
     input {expression=&self:MaterialExpressionFunctionInput_1}
    }
    {
     inputName=t
     input {expression=&self:MaterialExpressionFunctionInput_0}
    }
    {
     inputName=s
     input {expression=&self:MaterialExpressionVertexNormalWS_2}
    }
   ]
  }
  FunctionOutput(None=@n1)
 ]
}
