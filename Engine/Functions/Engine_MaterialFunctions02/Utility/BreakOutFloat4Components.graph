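# Material function BreakOutFloat4Components splits a float4 vector into its individual channels as scalar outputs.
# A single Vector4 input, Float4, feeds four component masks, each isolating one channel to drive the R, G, B and A outputs in that order.
# The input uses its preview value as the default when left unconnected.
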MaterialFunction /Engine/Functions/Engine_MaterialFunctions02/Utility/BreakOutFloat4Components
graph {
 properties {description=Enter a float 4 vector into the function and retrieve individual channels back through the scalar outputs.}
 nodes [
  n1 = FunctionInput() {bUsePreviewValueAsDefault=true, inputName=Float4, inputType=FunctionInput_Vector4}
  FunctionOutput(None=ComponentMask(None=@n1) {r=true}) {outputName=R}
  FunctionOutput(None=ComponentMask(None=@n1) {g=true}) {outputName=G, sortPriority=1}
  FunctionOutput(None=ComponentMask(None=@n1) {b=true}) {outputName=B, sortPriority=2}
  FunctionOutput(None=ComponentMask(None=@n1) {a=true}) {outputName=A, sortPriority=3}
 ]
}
