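# CameraDirectionVector is a material function, exposed to the library under Misc and Vectors, that outputs the world-space camera direction.
# It transforms a constant Z-forward unit vector from view space to world space and passes the result to its single output.
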MaterialFunction /Engine/Functions/Engine_MaterialFunctions02/Utility/CameraDirectionVector
graph {
 properties {bExposeToLibrary=true, description="World space vector of the camera direction. ", libraryCategoriesText [Misc,Vectors]}
 nodes [FunctionOutput(None=Transform(None=Constant3Vector() {constant=<rgba 0 0 1 0>}) {transformSourceType=TRANSFORMSOURCE_View})]
}
