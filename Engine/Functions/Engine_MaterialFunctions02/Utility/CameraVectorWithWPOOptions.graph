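# Material function, exposed to the library, that returns the normalized world-space camera vector with or without vertex shader offsets.
# It subtracts world position from camera position twice, once using the default world position and once using the position that excludes all shader offsets.
# A static bool input, "Use World Position Offsets" (default true), selects between the two through a static switch before normalization.
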MaterialFunction /Engine/Functions/Engine_MaterialFunctions02/Utility/CameraVectorWithWPOOptions
graph {
 properties {bExposeToLibrary=true, description=Retrieves the camera vector with or without vertex shader offsets., libraryCategoriesText [Misc,Vectors]}
 nodes [
  n1 = CameraPositionWS()
  n2 = Subtract(A=@n1, B=WorldPosition().XYZ)
  n3 = Subtract(A=@n1, B=WorldPosition() {worldPositionShaderOffset=WPT_ExcludeAllShaderOffsets}.XYZ)
  n4 = FunctionInput(Preview=StaticBool() {value=true}) {bUsePreviewValueAsDefault=true, description=Defaults to true, inputName=Use World Position Offsets, inputType=FunctionInput_StaticBool}
  FunctionOutput(None=Normalize(VectorInput=StaticSwitch(True=@n2, False=@n3, Value=@n4)))
 ]
}
