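# Material function, exposed to the library, that outputs a world-space reflection vector independent of the base shader's normal and default reflection vector.
# Optional inputs default to the world-space camera vector and vertex normal.
# It normalizes only the normal, then computes the standard mirror formula 2(C·N)N − C, where C is the camera vector.
# An empty normal-sampler texture object sits unconnected in the graph.
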MaterialFunction /Engine/Functions/Engine_MaterialFunctions02/Utility/CustomReflectionVector
graph {
 properties {
  bExposeToLibrary=true
  description=Input a normal to generate a reflection vector independant of the default reflection vector and the normals input on the base shader.
  libraryCategoriesText [Misc,Reflections]
 }
 nodes [
  n1 = FunctionInput(Preview=CameraVectorWS()) {
   bUsePreviewValueAsDefault=true
   description=Input a normalized vector or a normal map.
   inputName=CameraVector
   sortPriority=1
   previewValue=<0 0 1 1>
  }
  n2 = FunctionInput(Preview=VertexNormalWS()) {bUsePreviewValueAsDefault=true, description=Input a normalized vector or a normal map., inputName=Normal, previewValue=<0 0 1 1>}
  n3 = Normalize(VectorInput=@n2)
  n4 = Multiply(A=DotProduct(A=@n1, B=@n3), B=@n3)
  n5 = Constant() {r=2}
  FunctionOutput(None=Subtract(A=Multiply(A=@n4, B=@n5), B=@n1)) {description=World space reflection vector, outputName=Reflection Vector}
  TextureObject() {samplerType=SAMPLERTYPE_Normal, texture=null}
 ]
}
