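# Material function for debugging that draws an input scalar as readable digits in UV space, returning a single-channel mask.
# TextureCropping confines the text to a rectangle, which is split into digit columns either side of a decimal point.
# Powers of ten, Frac and Floor extract each column's digit to offset a DebugNumberStrip texture lookup.
# Masks hide leading zeros, while DebugNumberPeriod supplies the decimal point and minus sign.
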MaterialFunction /Engine/Functions/Engine_MaterialFunctions02/Utility/DebugScalarValues
graph {
 properties {
  bExposeToLibrary=true
  description=Enter a value into the material function and read its return value as a number in uv space.
  libraryCategoriesText [Misc,Debug]
 }
 nodes [
  n1 = FunctionInput(Preview=TextureCoordinate()) {bUsePreviewValueAsDefault=true, inputName=UVs, inputType=FunctionInput_Vector2, sortPriority=5, previewValue=<7 0 0 1>}
  n2 = FunctionInput() {
   bUsePreviewValueAsDefault=true
   inputName=DebugTextLocation RG_UpperRight BA_LowerLeft
   inputType=FunctionInput_Vector4
   sortPriority=6
   previewValue=<0 0.35 1 0.65>
  }
  n3 = MaterialFunctionCall!(
   UVs=@n1
   TextureIn
   UpperLeftCorner=ComponentMask(None=@n2) {g=true, r=true}
   LowerRightCorner=ComponentMask(None=@n2) {a=true, b=true}
   ExportFloat4
  ) {materialFunction=/Engine/Functions/Engine_MaterialFunctions01/Texturing/TextureCropping}
  n4 = Multiply(A=@n3.Crop UVs, B=Constant2Vector() {g=0.1, r=2}) {constA=2, constB=0.1}
  n5 = FunctionInput() {
   bUsePreviewValueAsDefault=true
   description=Max number of displayed integers.
   inputName=MaximumNumberOfDigits
   inputType=FunctionInput_Scalar
   sortPriority=1
   previewValue=<8 0 0 1>
  }
  n6 = Multiply(A=OneMinus(None=ComponentMask(None=@n4) {r=true}), B=@n5) {constB=10}
  n7 = Frac(None=@n6)
  n8 = FunctionInput(Preview=Constant() {r=128.5}) {bUsePreviewValueAsDefault=true, inputName=Number, inputType=FunctionInput_Scalar, previewValue=<128.5 0 0 1>}
  n9 = Abs(None=@n8)
  n10 = Add(A=Ceil(None=Clamp(None=@n6) {clampMode=CMODE_ClampMin, maxDefault=0})) {desc=1+}
  n11 = Divide(A=@n9, B=Power(Base=Constant() {r=10}, Exp=@n10) {desc="10-100,000,000"})
  n12 = Floor(None=Multiply(A=Frac(None=@n11)) {constB=100})
  n13 = ComponentMask(None=@n4) {g=true}
  n14 = Clamp(None=@n6) {clampMode=CMODE_ClampMax, maxDefault=0}
  n15 = Clamp(None=Subtract(A=Abs(None=@n14)) {desc=shift sampling to make room for decimal place}) {clampMode=CMODE_ClampMin, desc=Move 0 position and sample decimal points, maxDefault=0}
  n16 = Multiply(
   A=Add(A=@n9, B=Divide(A=@n9) {constB=1000000}) {constB=0.0001, desc="offset the floating point error precision error... "}
   B=Power(Base=Constant() {r=10}, Exp=Floor(None=@n15)) {desc="10-100,000,000"}
  )
  n17 = Floor(None=Multiply(A=Frac(None=@n16)) {constB=10})
  n18 = Clamp(None=Ceil(None=@n15))
  n19 = LinearInterpolate(A=Add(A=@n13, B=Divide(A=@n12) {constB=10}), B=Add(A=@n13, B=Divide(A=@n17) {constB=10}), Alpha=@n18)
  n20 = SphereMask(A=@n6, B=Constant() {r=-0.5}) {attenuationRadius=0.5, desc=mask .}
  n21 = Constant() {r=10}
  n22 = Add(A=@n10) {constB=3}
  n23 = Constant()
  n24 = If(A=Divide(A=@n9, B=Power(Base=@n21, Exp=@n22)), B=@n23, "A > B"=Constant() {r=1}, "A == B"=@n23, "A < B"=@n23)
  n25 = Multiply(
   A=TextureSample!(UVs=AppendVector(A=@n7, B=@n19), Tex, Level, Apply View MipBias) {mipValueMode=TMVM_MipLevel, texture=/Engine/Functions/Engine_MaterialFunctions02/ExampleContent/DebugNumberStrip}.R
   B=Multiply(A=OneMinus(None=@n20) {desc=remove decimal point}, B=Add(A=@n24, B=@n18)) {desc=remove 0's on the int side}
  ) {desc=horizontal mask}
  n26 = Multiply(A=Add(A=@n5, B=@n5), B=ComponentMask(None=Subtract(A=@n3.Crop UVs) {constB=0.5}) {r=true})
  n27 = ComponentMask(None=@n3.Crop UVs) {g=true}
  n28 = TextureSample(UVs=AppendVector(A=Add(B=@n26) {constA=0.5}, B=@n27)) {mipValueMode=TMVM_MipLevel, texture=/Engine/Functions/Engine_MaterialFunctions02/ExampleContent/Textures/DebugNumberPeriod}
  n29 = Constant() {r=1}
  n30 = Add(A=@n8)
  n31 = Multiply(A=@n8) {constB=-1}
  n32 = Saturate(None=Multiply(A=Ceil(None=@n30), B=Ceil(None=@n31)))
  n33 = Divide(A=@n9, B=Power(Base=Constant() {r=10}, Exp=Add(A=@n10) {constB=2}))
  n34 = Constant()
  n35 = Constant() {r=1}
  n36 = Constant()
  n37 = TextureSample(
   UVs=AppendVector(A=Frac(None=@n6), B=@n27)
   Level=MaterialFunctionCall!(UVs=AppendVector(A=@n6, B=@n27), Texture Size=Constant2Vector() {g=128, r=128}) {materialFunction=/Engine/Functions/Engine_MaterialFunctions01/Texturing/ComputeMipLevel}.MipLevel
  ) {
   desc=Negative Symbol
   mipValueMode=TMVM_MipLevel
   texture=&DebugNumberPeriod
  }
  n38 = Constant()
  n39 = Multiply(
   A=If(A=If(A=@n33, B=@n34, "A > B"=@n35, "A == B"=@n34, "A < B"=@n34), B=@n24, "A > B"=@n37.G, "A == B"=@n36, "A < B"=@n36)
   B=If(A=@n8, B=@n38, "A > B"=@n38, "A == B"=@n38, "A < B"=Constant() {r=1}) {desc=- or positive number, equalsThreshold=0}
  ) {desc=solves negative symbol}
  n40 = Constant() {r=0.5}
  n41 = Multiply(A=@n37.G, B=Multiply(A=@n32, B=SphereMask(A=@n6, B=@n40) {attenuationRadius=0.5, desc=location mask}))
  n42 = Multiply(
   A=Add(A=Add(A=@n25, B=@n28.R) {desc=Decimal point}, B=If(A=@n29, B=@n32, "A > B"=@n39, "A == B"=@n41, "A < B"=@n39)) {desc=add sign}
   B=@n3.Crop Mask
  ) {desc=vertical mask}
  FunctionOutput(None=ComponentMask(None=Clamp(None=@n42)) {r=true})
 ]
}
