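# Debug material function that renders the four components of a Vector4 as readable numbers when the output is previewed.
# It crops UVs to a text region, stretches V fourfold and offsets it per component, stacking four DebugScalarValues calls with shared spacing and digit limit.
# GreyScaleOutput sums the four results with a dot product.
# ColorCodedOutput shows X, Y, Z in red, green, blue, adds W as white and carries W in alpha.
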MaterialFunction /Engine/Functions/Engine_MaterialFunctions02/Utility/DebugFloat4Values
graph {
 properties {
  bExposeToLibrary=true
  description=Plug a vector into the function and then preview the output to view the numeric values of the input vector.
  libraryCategoriesText [Misc,Debug]
 }
 nodes [
  n1 = FunctionInput() {
   bUsePreviewValueAsDefault=true
   description=Enter a number to view.
   inputName=Vector4
   inputType=FunctionInput_Vector4
   previewValue=<1 2 3 4>
  }
  n2 = MaterialFunctionCall(Float4=@n1) {materialFunction=&BreakOutFloat4Components}
  n3 = FunctionInput() {
   bUsePreviewValueAsDefault=true
   description=Max number of digits on either side of the decimal place.
   inputName=MaximumNumberOfDigits
   inputType=FunctionInput_Scalar
   sortPriority=1
   previewValue=<5 0 0 1>
  }
  n4 = FunctionInput(Preview=TextureCoordinate()) {bUsePreviewValueAsDefault=true, inputName=UVs, inputType=FunctionInput_Vector2, sortPriority=2, previewValue=<7 0 0 1>}
  n5 = FunctionInput() {
   bUsePreviewValueAsDefault=true
   inputName=DebugTextLocation RG_UpperRight BA_LowerLeft
   inputType=FunctionInput_Vector4
   sortPriority=3
   previewValue=<0.3 0.3 0.7 0.7>
  }
  n6 = Multiply(
   A=MaterialFunctionCall!(
    UVs=@n4
    TextureIn
    UpperLeftCorner=ComponentMask(None=@n5) {g=true, r=true}
    LowerRightCorner=ComponentMask(None=@n5) {a=true, b=true}
    ExportFloat4
   ) {materialFunction=/Engine/Functions/Engine_MaterialFunctions01/Texturing/TextureCropping}.Crop UVs
   B=Constant2Vector() {g=4, r=1}
  )
  n7 = FunctionInput() {
   bUsePreviewValueAsDefault=true
   description=RG_UpperRight BA_LowerLeft
   inputName=Component Spacing
   inputType=FunctionInput_Vector4
   sortPriority=5
   previewValue=<0 0.35 1 0.65>
  }
  n8 = Subtract(A=@n6, B=Constant2Vector() {g=1})
  n9 = Constant2Vector() {g=2}
  n10 = AppendVector(
   A=AppendVector(
    A=MaterialFunctionCall(Number=@n2.R, MaximumNumberOfDigits=@n3, UVs=@n6, DebugTextLocation RG_UpperRight BA_LowerLeft=@n7) {materialFunction=&DebugScalarValues}.Result
    B=MaterialFunctionCall(Number=@n2.G, MaximumNumberOfDigits=@n3, UVs=@n8, DebugTextLocation RG_UpperRight BA_LowerLeft=@n7) {materialFunction=&DebugScalarValues}.Result
   )
   B=MaterialFunctionCall(Number=@n2.B, MaximumNumberOfDigits=@n3, UVs=Subtract(A=@n6, B=@n9), DebugTextLocation RG_UpperRight BA_LowerLeft=@n7) {materialFunction=&DebugScalarValues}.Result
  )
  n11 = MaterialFunctionCall(Number=@n2.A, MaximumNumberOfDigits=@n3, UVs=Subtract(A=@n6, B=Constant2Vector() {g=3}), DebugTextLocation RG_UpperRight BA_LowerLeft=@n7) {materialFunction=&DebugScalarValues}
  FunctionOutput(None=DotProduct(A=AppendVector(A=@n10, B=@n11.Result), B=Constant4Vector() {constant=<rgba 1 1 1 1>}.RGBA)) {desc=Alpha, outputName=GreyScaleOutput, sortPriority=1}
  FunctionOutput(None=AppendVector(A=Add(A=@n10, B=@n11.Result), B=@n11.Result)) {outputName=ColorCodedOutput}
 ]
}
