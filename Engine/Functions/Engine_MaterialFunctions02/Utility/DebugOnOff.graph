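# DebugOnOff is a material function in the Misc and Debug library categories that outputs a blinking on/off value for A/B comparisons in materials.
# It feeds Time into a Sine with a period of two, then applies Ceil and Clamp to turn the wave into a hard 0 or 1 square toggle.
# It takes no inputs, so the switch runs automatically.
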MaterialFunction /Engine/Functions/Engine_MaterialFunctions02/Utility/DebugOnOff
graph {
 properties {bExposeToLibrary=true, description=Useful for doing AB comparrison in material, libraryCategoriesText [Misc,Debug]}
 nodes [
  n1 = Sine(None=Time()) {period=2}
  FunctionOutput(None=Clamp(None=Ceil(None=@n1)))
 ]
}
