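# DebugTimeSine is a library material function for debugging that shows how a scalar value behaves as it oscillates between 0 and 1.
# It multiplies Time by a scalar Speed input, which falls back to its preview value when unconnected, and feeds the result into a Sine.
# A ConstantBiasScale at default settings remaps the wave from the -1 to 1 range into 0 to 1 for output.
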MaterialFunction /Engine/Functions/Engine_MaterialFunctions02/Utility/DebugTimeSine
graph {
 properties {bExposeToLibrary=true, description=useful for seeing what a scalar value does between 0 and 1, libraryCategoriesText [Misc,Debug]}
 nodes [
  n1 = FunctionInput() {bUsePreviewValueAsDefault=true, inputName=Speed, inputType=FunctionInput_Scalar, previewValue=<0.5 0 0 1>}
  n2 = Multiply(A=Time(), B=@n1)
  FunctionOutput(None=ConstantBiasScale(None=Sine(None=@n2)))
 ]
}
