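# DepthFromWorldPosition is a library material function that returns depth, comparable to Pixel Depth or Scene Depth, for an arbitrary world position.
# It subtracts the camera position from the input position and dots the result with the camera direction vector, giving distance along the view axis rather than radial distance.
# The World Position input defaults to the world position excluding all shader offsets.
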MaterialFunction /Engine/Functions/Engine_MaterialFunctions02/Utility/DepthFromWorldPosition
graph {
 properties {
  bExposeToLibrary=true
  description=Returns the same value that Pixel Depth or Scene Depth Return but using an arbitrary world position.
  libraryCategoriesText [Misc,Depth]
 }
 nodes [
  n1 = FunctionInput(Preview=WorldPosition() {worldPositionShaderOffset=WPT_ExcludeAllShaderOffsets}.XYZ) {bUsePreviewValueAsDefault=true, inputName=World Position}
  n2 = CameraPositionWS()
  FunctionOutput(None=DotProduct(
    A=MaterialFunctionCall() {materialFunction=&CameraDirectionVector}.Result
    B=Subtract(A=@n1, B=@n2)
   ))
 ]
}
