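# DynamicBranch is a library material utility function that picks between two color inputs based on a scalar Alpha.
# An If node compares Alpha to zero with no equality threshold, outputting Color Input 1 when greater and Color Input 0 when equal or less.
# All inputs use preview values as defaults.
# Its description presents it as a dynamic branch for skipping costly masked instructions or texture samples.
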MaterialFunction /Engine/Functions/Engine_MaterialFunctions02/Utility/DynamicBranch
graph {
 properties {
  bExposeToLibrary=true
  description="The code connected to the \"Color Input 1\" will be executed if the alpha is greater than 0 and the \"Color Input 0\" input code will be evaluated  if the alpha is 0. Dynamic branches are best used when masking 20 or more instructions or a texture sample."
 }
 nodes [
  n1 = FunctionInput() {bUsePreviewValueAsDefault=true, inputName=Color Input 1, sortPriority=2, previewValue=<1 1 1 1>}
  n2 = FunctionInput() {bUsePreviewValueAsDefault=true, inputName=Color Input 0, sortPriority=1}
  FunctionOutput(None=If(
    A=FunctionInput() {bUsePreviewValueAsDefault=true, inputName=Alpha, inputType=FunctionInput_Scalar}
    B=Constant()
    "A > B"=@n1
    "A == B"=@n2
    "A < B"=@n2
   ) {equalsThreshold=0})
 ]
}
