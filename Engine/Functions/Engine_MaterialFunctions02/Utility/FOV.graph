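# Material function exposed to the library that returns the current camera's field of view.
# Custom HLSL nodes read View.ClipToView[0][0] for a raw output, and convert it with atan times two for radians, which is scaled to degrees for a third output.
# Each output is masked to the red channel.
# A two-axis Custom node covering both matrix diagonal terms feeds no output.
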MaterialFunction /Engine/Functions/Engine_MaterialFunctions02/Utility/FOV
graph {
 properties {bExposeToLibrary=true, description=Returns the current cameras FOV setting}
 nodes [
  Custom!(None) {code="atan(float2(View.ClipToView[0][0], View.ClipToView[1][1]))*2", description=GrabFOV, outputType=CMOT_Float2}
  n1 = Custom() {code="atan(View.ClipToView[0][0])*2", description=GrabFOV, outputType=CMOT_Float2}
  FunctionOutput(None=ComponentMask(None=Multiply(A=@n1) {constB=57.2958}) {r=true}) {outputName=FOV angle in degrees, sortPriority=2}
  FunctionOutput(None=ComponentMask(None=@n1) {r=true}) {outputName=FOV in radians, sortPriority=1}
  FunctionOutput(None=ComponentMask(None=Custom() {code="View.ClipToView[0][0]", description=GrabFOV, outputType=CMOT_Float2}) {r=true}) {description=Returns the fov before expensive conversions are run., outputName="FOV "}
 ]
}
