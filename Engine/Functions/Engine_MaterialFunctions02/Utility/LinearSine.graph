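# LinearSine is a library material function that turns a scalar into a repeating triangle wave with a configurable period.
# It divides the value, optionally shifted a quarter period to match sine phase, by Period, then folds the fraction into a 0-1 ramp up and down.
# A smoothstep cubic adds a rounded output, and a static switch can remap both waves to -1 to 1.
# A Direction output flags rising or falling slope.
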MaterialFunction /Engine/Functions/Engine_MaterialFunctions02/Utility/LinearSine
graph {
 properties {
  bExposeToLibrary=true
  description=A sine-like function that transitions linearly in the 0-1 range using a period of 1.
  libraryCategoriesText [Misc,Math]
 }
 nodes [
  Constant() {r=1}
  n1 = FunctionInput(Preview=Constant() {r=0.5}) {
   bUsePreviewValueAsDefault=true
   description=Insert a gradient or float value to operate on
   inputName=Value
   inputType=FunctionInput_Scalar
   previewValue=<1 0 0 1>
  }
  n2 = FunctionInput() {
   bUsePreviewValueAsDefault=true
   description=Insert float value. Defaults to 1.
   inputName=Period
   inputType=FunctionInput_Scalar
   sortPriority=2
   previewValue=<1 0 0 1>
  }
  n3 = Add(A=@n1, B=Multiply(A=Constant() {r=0.25}, B=@n2))
  n4 = FunctionInput(Preview=StaticBool() {value=true}) {
   bUsePreviewValueAsDefault=true
   description=Aligns phase with standard sine function. Defaults to true.
   inputName=Sine Phase
   inputType=FunctionInput_StaticBool
   sortPriority=5
   previewValue=<1 0 0 1>
  }
  n5 = Frac(None=Divide(A=StaticSwitch(True=@n3, False=@n1, Value=@n4) {defaultValue=true}, B=@n2))
  n6 = Constant() {r=2}
  n7 = Multiply(A=@n5, B=@n6)
  n8 = Floor(None=@n7)
  n9 = LinearInterpolate(A=@n7, B=Multiply(A=@n6, B=OneMinus(None=@n5)), Alpha=@n8)
  n10 = FunctionInput(Preview=StaticBool()) {
   bUsePreviewValueAsDefault=true
   description="Constant bias scale the sine wave to a -1 to 1 range if true.  The default value is false. "
   inputName=-1 to 1
   inputType=FunctionInput_StaticBool
   sortPriority=3
  }
  FunctionOutput(None=StaticSwitch(True=ConstantBiasScale(None=@n9) {bias=-0.5, scale=2}, False=@n9, Value=@n10)) {description=Returns 0-1 linear sine with a period of 1, outputName=Linear Sine}
  n11 = Constant() {r=3}
  n12 = Multiply(A=@n9, B=Constant() {r=2})
  n13 = Multiply(A=Multiply(A=Subtract(A=@n11, B=@n12), B=@n9), B=@n9)
  FunctionOutput(None=StaticSwitch(True=ConstantBiasScale(None=@n13) {bias=-0.5, scale=2}, False=@n13, Value=@n10)) {description=Returns 0-1 linear sine with rounded edges with a period of 1, outputName="Rounded Linear Sine ", sortPriority=1}
  FunctionOutput(None=LinearInterpolate(Alpha=@n8) {constA=1, constB=0}) {
   desc=Direction
   description="Returns 0 or 1 based on whether the slope of the sine wave is traveling downward (0) or upward(1)."
   outputName=Direction
   sortPriority=3
  }
 ]
}
