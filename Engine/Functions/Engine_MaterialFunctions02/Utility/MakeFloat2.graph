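# MakeFloat2 is an engine utility material function that builds a two-component vector from scalar inputs.
# Two scalar inputs, X and Y, feed an AppendVector node whose result is the function's single output.
# Both inputs use their preview values as defaults, so either can be left unconnected.
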MaterialFunction /Engine/Functions/Engine_MaterialFunctions02/Utility/MakeFloat2
graph {
 properties {description=Creates a float 2 vector from a series of scalar inputs.}
 nodes [FunctionOutput(None=AppendVector(
    A=FunctionInput() {bUsePreviewValueAsDefault=true, inputName=X, inputType=FunctionInput_Scalar}
    B=FunctionInput() {bUsePreviewValueAsDefault=true, inputName=Y, inputType=FunctionInput_Scalar, sortPriority=1}
   ))]
}
