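# MakeFloat3 is an engine utility material function that builds a float3 vector from three scalar inputs.
# It chains two AppendVector nodes, joining X with Y and then appending Z, and sends the result to a single output.
# All three inputs use their preview value as the default, so each can be left unconnected.
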MaterialFunction /Engine/Functions/Engine_MaterialFunctions02/Utility/MakeFloat3
graph {
 properties {description=Creates a float 3 vector from a series of scalar inputs.}
 nodes [
  n1 = FunctionInput() {bUsePreviewValueAsDefault=true, inputName=X, inputType=FunctionInput_Scalar}
  n2 = FunctionInput() {bUsePreviewValueAsDefault=true, inputName=Y, inputType=FunctionInput_Scalar, sortPriority=1}
  FunctionOutput(None=AppendVector(
    A=AppendVector(A=@n1, B=@n2)
    B=FunctionInput() {bUsePreviewValueAsDefault=true, inputName=Z, inputType=FunctionInput_Scalar, sortPriority=2}
   ))
 ]
}
