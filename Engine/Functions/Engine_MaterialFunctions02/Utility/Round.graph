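# Material function that rounds a scalar to the nearest whole number, exposed to the library under Misc and Math.
# An If compares the absolute fractional part against one half and picks Ceil when above, Floor when below.
# Exact halves lerp between Ceil and Floor by the remapped Sign of the input, rounding toward zero; a node note says this is meant to match HLSL.
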MaterialFunction /Engine/Functions/Engine_MaterialFunctions02/Utility/Round
graph {
 properties {bExposeToLibrary=true, description=Rounds a number to the next closest whole number, libraryCategoriesText [Misc,Math]}
 nodes [
  n1 = FunctionInput() {bUsePreviewValueAsDefault=true, description=Rounds to the nearest whole number., inputType=FunctionInput_Scalar}
  n2 = Frac(None=@n1)
  n3 = Ceil(None=@n1)
  n4 = Floor(None=@n1)
  n5 = ConstantBiasScale(None=MaterialFunctionCall!(In=@n1) {materialFunction=&/Sign}.Result)
  FunctionOutput(None=If(
    A=Abs(None=@n2)
    B=Constant() {r=0.5}
    "A > B"=@n3
    "A == B"=LinearInterpolate(A=@n3, B=@n4, Alpha=@n5) {desc=Done to match the hlsl functionality}
    "A < B"=@n4
   ))
 ]
}
