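# Sine_Remapped is a library material function that remaps a sine wave to oscillate between two user-specified colors or vectors.
# The Sine Phase scalar feeds a Sine node, and ConstantBiasScale at its default settings shifts that wave into the 0-1 range.
# The result is the Lerp alpha between Value 1 and Value 2, output as Remapped Sine with a period of 1.
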MaterialFunction /Engine/Functions/Engine_MaterialFunctions02/Utility/Sine_Remapped
graph {
 properties {
  bExposeToLibrary=true
  description=remaps the output of a sine wave to two colors/vectors specified by the user.
  libraryCategoriesText [Misc,Math]
 }
 nodes [
  n1 = Sine(None=FunctionInput() {bUsePreviewValueAsDefault=true, inputName=Sine Phase, inputType=FunctionInput_Scalar})
  FunctionOutput(None=LinearInterpolate(
    A=FunctionInput() {bUsePreviewValueAsDefault=true, inputName=Value 1, sortPriority=1}
    B=FunctionInput() {bUsePreviewValueAsDefault=true, inputName=Value 2, sortPriority=2, previewValue=<1 1 1 1>}
    Alpha=ConstantBiasScale(None=@n1)
   )) {description=Returns 0-1 linear sine with a period of 1, outputName=Remapped Sine}
 ]
}
