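# Swizzle is a library-exposed material function utility that reorders vector components.
# A Vector2 XY input, previewing texture coordinates, is split with component masks and appended in reverse to output YX.
# A separate Vector3 XYZ input is split and re-appended to output YXZ, swapping the first two channels and keeping Z in place.
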MaterialFunction /Engine/Functions/Engine_MaterialFunctions02/Utility/Swizzle
graph {
 properties {bExposeToLibrary=true}
 nodes [
  n1 = FunctionInput(Preview=TextureCoordinate()) {bUsePreviewValueAsDefault=true, description=Input the Vector2 you want swizzled, inputName=XY, inputType=FunctionInput_Vector2}
  FunctionOutput(None=AppendVector(A=ComponentMask(None=@n1) {g=true}, B=ComponentMask(None=@n1) {r=true})) {outputName=YX}
  n2 = FunctionInput(Preview=Constant3Vector()) {bUsePreviewValueAsDefault=true, description=Input Vector3 you want Swizzled, inputName=XYZ}
  n3 = ComponentMask(None=@n2) {g=true}
  n4 = ComponentMask(None=@n2) {r=true}
  FunctionOutput(None=AppendVector(A=AppendVector(A=@n3, B=@n4), B=ComponentMask(None=@n2) {b=true})) {outputName=YXZ}
 ]
}
