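# VectorLength is a library-exposed material function that returns the length of a vector.
# Separate Vector 3 and Vector 2 inputs each feed a Distance node measured against one shared zero constant, giving the V3 Length and V2 Length outputs.
# Both inputs use their preview value as the default, so either can be left unconnected.
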MaterialFunction /Engine/Functions/Engine_MaterialFunctions02/Utility/VectorLength
graph {
 properties {bExposeToLibrary=true, description=Returns a vectors length.}
 nodes [
  n1 = Constant()
  FunctionOutput(None=Distance(A=@n1, B=FunctionInput() {bUsePreviewValueAsDefault=true, inputName=Vector 3})) {outputName=V3 Length}
  FunctionOutput(None=Distance(A=FunctionInput() {bUsePreviewValueAsDefault=true, inputName=Vector 2, inputType=FunctionInput_Vector2, sortPriority=1}, B=@n1)) {outputName=V2 Length, sortPriority=1}
 ]
}
