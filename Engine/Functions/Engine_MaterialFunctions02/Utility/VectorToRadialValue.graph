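# Material function that converts a 2D vector or UVs into radial coordinates, exposed to the library under Misc and Math.
# A Custom node runs atan2, and dividing by 2π then taking Frac wraps the angle to 0–1, while squared components yield the length.
# It outputs angle, linear distance and both combined, with a static bool swizzling channel order.
# The input defaults to centered -1 to 1 UVs, and a leftover If branch feeds no output.
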MaterialFunction /Engine/Functions/Engine_MaterialFunctions02/Utility/VectorToRadialValue
graph {
 properties {
  bExposeToLibrary=true
  description=The input coordinates or vector will be transformed into radial coordinates. In the case of a vector the angle will be output in one channel and the length of the vector in the other.
  libraryCategoriesText [Misc,Math]
 }
 nodes [
  n1 = FunctionInput(Preview=Subtract(A=TextureCoordinate() {uTiling=2, vTiling=2}, B=Constant() {r=1})) {
   bUsePreviewValueAsDefault=true
   description="Insert a vector 2 value to retrieve the vectors angle or insert uvs to retrieve radial coordinates. Note that (0,0) is not an acceptable number.  "
   inputName=Vector or UVs
   inputType=FunctionInput_Vector2
   previewValue=<1 1 0 1>
  }
  MaterialExpressionComponentMask_20 = ComponentMask(None=@n1) {r=true}
  MaterialExpressionComponentMask_21 = ComponentMask(None=@n1) {g=true}
  n2 = Custom!(x=@MaterialExpressionComponentMask_20, y=@MaterialExpressionComponentMask_21) {
   code="return (atan2(y,x));"
   description=atan2
   outputType=CMOT_Float1
   inputs [
    {inputName=x, input {expression=&self:MaterialExpressionComponentMask_20}}
    {inputName=y, input {expression=&self:MaterialExpressionComponentMask_21}}
   ]
  }
  n3 = ComponentMask(None=@n2) {r=true}
  n4 = Constant() {r=6.2832}
  n5 = Divide(A=@n3, B=@n4)
  n6 = Frac(None=@n5)
  n7 = Multiply(A=@n1, B=@n1)
  n8 = SquareRoot(None=Add(A=ComponentMask(None=@n7) {r=true}, B=ComponentMask(None=@n7) {g=true}))
  n9 = FunctionInput(Preview=StaticBool() {value=true}) {
   bUsePreviewValueAsDefault=true
   description="Switch the order of the channels exported from the \"radial coordinates\" output."
   inputName=Swizzle Coordinate Output
   inputType=FunctionInput_StaticBool
  }
  FunctionOutput(None=StaticSwitch(True=AppendVector(A=@n6, B=@n8), False=AppendVector(A=@n8, B=@n6), Value=@n9)) {
   description=The input coordinates or vector will be transformed into radial coordinates. In the case of a vector the angle will be output in one channel and the length of the vector in the other.
   outputName=Radial Coordinates
  }
  FunctionOutput(None=@n6) {
   description=This will return the angle of of the input vector or a radial gradient in the case of input uvs.
   outputName=Vector Converted to Angle
   sortPriority=1
  }
  FunctionOutput(None=@n8) {
   description=Returns the linear length of the input vector or a gradient of distances in the case of uvs.
   outputName=Linear Distance
   sortPriority=2
  }
  If(A=@n3, B=Constant(), "A > B"=@n5, "A < B"=Divide(A=Add(A=@n3, B=@n4), B=@n4))
 ]
}
